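# Animation Blueprint for the Innovator Female glider, animating deploy, glide and close during skydiving.
# A Locomotion state machine follows parachute state or force-open, with front-end variants for cosmetic preview.
# Curve-weighted additive leans from acceleration, an aim-driven turn blendspace and jitter layer on top, then DefaultSlot.
# Notifies toggle glider visibility and mark full deployment; update logic computes billow values, but the billow player is unconnected to the output.
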AnimBlueprint /BRCosmetics/Animation/Game/MainPlayer/Skydive/ParaGlide/Innovator_Female/Innovator_Female_Glider_AnimBP
target_skeleton=/BRCosmetics/Gadgets/Assets/VinderTech_GliderChute/Glider_Innovator_Female/Meshes/Innovator_Female_Skeleton
graphs [
 AnimGraph {nodes [
   Apply Additive() {alphaBoolBlend {blendOption=Linear}}
   n1 = Break Fort Anim Input Skydiving!(FortAnimInput_Skydiving:"Fort Anim Input Skydiving Structure (by ref)"=Get SkydivingState() -> (SkydivingState:Fort Anim Input Skydiving Structure)) -> (
    ParaGilderRootOffset:Vector=~
    ParaGliderRootRotationOverride:Rotator=~
    bUseParaGlideRootModifier:bool=~
    bIsSkydivingFromLaunchPad:bool=~
    bIsSkydivingFromBus:bool=~
    bIsInVortex:bool=~
    bIsDiving:bool=~
    bIsParachuteOpen:bool=~
    bIsSkydiving:bool=~
    bIsParachuteLeaning:bool=~
    bIsSkydiveLeaning:bool=~
    bIsLeaning:bool=~
    bIsSkydiveDiveMode:bool=~
    bPlaySkydiveDrift:bool=~
    bSkydiveDriftDelayActive:bool=~
    bSkydiveDriftAnimAllowed:bool=~
    bIsGliderRight:bool=~
    bIsGliderCenter:bool=~
    bIsGliderLeft:bool=~
    bIsGliderForward:bool=~
    bIsGliderBack:bool=~
    bHasDeployedAndIsStandingUpright:bool=~
    LocalAccelForward:float=~
    LocalAccelRight:float=~
    LocalVelocityRight:float=~
    SkydiveAimPitch:float=~
    SkydiveAimYaw:float=~
    DeployChuteAnimRate:float=~
    SkydiveDriftAnimRate:float=~
    SkydiveDriftAnimRateCurrent:float=~
    SkydiveFidgetAnimRate:float=~
    SkydiveFidgetAnimRateCurrent:float=~
    SkydiveAdditiveAlpha:float=~
    SkydiveDriftDelay:float=~
    SkydiveDriftAnim:int
    SkydiveDriftAnimMax:int
    LaunchpadAnim:int
    LocalAccelDir:ESkydivingDirection Enum
    DirectionLast:ESkydivingDirection Enum
    PoseAdditive:Anim Sequence Object Reference
   ) {is=title K2Node_BreakStruct}
   Blendspace Player 'BS_Innovator_Female_Lean_GLIDER'(X=@n1.LocalAccelRight, Y=@n1.LocalAccelForward) {blendSpace=&/BS_Innovator_Female_Lean_GLIDER}
   AnimGraphNode_StateMachine_0 = Locomotion() {is=title AnimGraphNode_StateMachine}
   AnimGraphNode_StateMachine_1 = Lean() {is=title AnimGraphNode_StateMachine}
   n2 = Apply Additive(
    Base=Apply Additive!(Base=@AnimGraphNode_StateMachine_0, Additive=@AnimGraphNode_StateMachine_1, AlphaCurveName:Name=ParaGlide_Lean) -> (Pose) {alphaInputType=Curve, alphaBoolBlend {blendOption=Linear}, alphaScaleBiasClamp {bias=1, scale=-1}}
    Additive=Blendspace Player 'BS_Innovator_Female_Turn_GLIDER'(X=@n1.SkydiveAimYaw, Y=@n1.SkydiveAimPitch) {blendSpace=&/BS_Innovator_Female_Turn_GLIDER}
   ) {alphaBoolBlend {blendOption=Linear}}
   n3 = "Sequence Player (additive) 'Innovator_Female_Jitter_GLIDER'"() {sequence=&/Innovator_Female_Jitter_GLIDER}
   Output Pose(Result=Slot 'DefaultSlot'(Source=Apply Additive!(Base=@n2, Additive=@n3, Alpha:float) -> (Pose) {alphaBoolBlend {blendOption=Linear}})) {layerGroup=None}
   Sequence Player!(PlayRate:float=1) -> (Pose)
   Sequence Player!(
    Sequence:Anim Sequence Base Object Reference=Get Billow_Anim() -> (Billow_Anim:Anim Sequence Base Object Reference)
    PlayRate=Get Billow_Speed() -> (Billow_Speed:double)
   ) -> (Pose)
   Get Jitter_Speed() -> (Jitter_Speed:double)
   Get Billow_Alpha() -> (Billow_Alpha:double)
  ]}
 Locomotion {
  entry=Closed
  path=AnimGraph.AnimGraphNode_StateMachine_0.Locomotion
  ids=id
  nodes [
   AnimStateNode_0 = Idle_Open() {is=title AnimStateNode}
   AnimStateNode_1 = Open_Into() {is=title AnimStateNode, stateEntered {guid=#1, notifyName=DeployEntered}}
   AnimStateNode_2 = Closed() {is=title AnimStateNode, stateEntered {guid=#2, notifyName=ClosedEntered}}
   AnimStateTransitionNode_0 = AnimStateTransitionNode() {bAutomaticRuleBasedOnSequencePlayerInState=true, blendMode=Linear, crossfadeDuration=0.1}
   AnimStateTransitionNode_1 = AnimStateTransitionNode() {crossfadeDuration=0}
   AnimStateTransitionNode_2 = AnimStateTransitionNode() {crossfadeDuration=0}
   AnimStateTransitionNode_3 = AnimStateTransitionNode() {crossfadeDuration=0, priorityOrder=2}
  ]
  links [
   {
    dst=Open_Into
    rule=AnimGraph.AnimGraphNode_StateMachine_0.Locomotion.AnimStateTransitionNode_1.Transition
    src=Closed
    via=AnimStateTransitionNode_1
   }
   {
    dst=Closed
    rule=AnimGraph.AnimGraphNode_StateMachine_0.Locomotion.AnimStateTransitionNode_2.Transition
    src=Idle_Open
    via=AnimStateTransitionNode_2
   }
   {
    dst=Closed
    rule=AnimGraph.AnimGraphNode_StateMachine_0.Locomotion.AnimStateTransitionNode_3.Transition
    src=Open_Into
    via=AnimStateTransitionNode_3
   }
   {
    dst=Idle_Open
    rule=AnimGraph.AnimGraphNode_StateMachine_0.Locomotion.AnimStateTransitionNode_0.Transition
    src=Open_Into
    via=AnimStateTransitionNode_0
   }
  ]
 }
 Idle_Open {
  path=AnimGraph.AnimGraphNode_StateMachine_0.Locomotion.AnimStateNode_0.Idle_Open
  nodes [
   Apply Additive!(
    Base
    Additive
    bAlphaBoolEnabled:bool=Get IsCosmeticPreview!(self:Fort Player Parachute Object Reference=Get OwnerGlider() -> (OwnerGlider:Fort Player Parachute Object Reference)) -> (IsCosmeticPreview:bool)
   ) -> (Pose) {alphaInputType=Bool, alphaBoolBlend {blendOption=Linear}}
   n1 = Break Fort Anim Input Skydiving(FortAnimInput_Skydiving=Get SkydivingState()) -> (
    ParaGilderRootOffset=~
    ParaGliderRootRotationOverride=~
    bUseParaGlideRootModifier=~
    bIsSkydivingFromLaunchPad=~
    bIsSkydivingFromBus=~
    bIsInVortex=~
    bIsDiving=~
    bIsParachuteOpen=~
    bIsSkydiving=~
    bIsParachuteLeaning=~
    bIsSkydiveLeaning=~
    bIsLeaning=~
    bIsSkydiveDiveMode=~
    bPlaySkydiveDrift=~
    bSkydiveDriftDelayActive=~
    bSkydiveDriftAnimAllowed=~
    bIsGliderRight=~
    bIsGliderCenter=~
    bIsGliderLeft=~
    bIsGliderForward=~
    bIsGliderBack=~
    bHasDeployedAndIsStandingUpright=~
    LocalAccelForward=~
    LocalAccelRight=~
    LocalVelocityRight=~
    SkydiveAimPitch=~
    SkydiveAimYaw=~
    DeployChuteAnimRate=~
    SkydiveDriftAnimRate=~
    SkydiveDriftAnimRateCurrent=~
    SkydiveFidgetAnimRate=~
    SkydiveFidgetAnimRateCurrent=~
    SkydiveAdditiveAlpha=~
    SkydiveDriftDelay=~
   )
   n2 = Blendspace Player 'BS_Innovator_Female_GLIDER'(X=@n1.SkydiveAimYaw, Y=@n1.SkydiveAimPitch) {
    blendSpace=&/BS_Innovator_Female_GLIDER
    groupName=Glide
    method=SyncGroup
   }
   n3 = Get OwnerGlider()
   Output Animation Pose(Result=Blend Poses by bool(
     BlendPose_0=Sequence Player 'Innovator_Female_Glide_FrontEnd_GLIDER'() {sequence=&/Innovator_Female_Glide_FrontEnd_GLIDER}
     BlendPose_1=@n2
     bActiveValue=Get IsCosmeticPreview(self=@n3)
    ) {blendType=Linear})
  ]
 }
 Open_Into {
  path=AnimGraph.AnimGraphNode_StateMachine_0.Locomotion.AnimStateNode_1.Open_Into
  nodes [
   Apply Additive(bAlphaBoolEnabled=Get IsCosmeticPreview(self=Get OwnerGlider())) {alphaInputType=Bool, alphaBoolBlend {blendOption=Linear}}
   n1 = Break Fort Anim Input Skydiving(FortAnimInput_Skydiving=Get SkydivingState()) -> (
    ParaGilderRootOffset=~
    ParaGliderRootRotationOverride=~
    bUseParaGlideRootModifier=~
    bIsSkydivingFromLaunchPad=~
    bIsSkydivingFromBus=~
    bIsInVortex=~
    bIsDiving=~
    bIsParachuteOpen=~
    bIsSkydiving=~
    bIsParachuteLeaning=~
    bIsSkydiveLeaning=~
    bIsLeaning=~
    bIsSkydiveDiveMode=~
    bPlaySkydiveDrift=~
    bSkydiveDriftDelayActive=~
    bSkydiveDriftAnimAllowed=~
    bIsGliderRight=~
    bIsGliderCenter=~
    bIsGliderLeft=~
    bIsGliderForward=~
    bIsGliderBack=~
    bHasDeployedAndIsStandingUpright=~
    LocalAccelForward=~
    LocalAccelRight=~
    LocalVelocityRight=~
    SkydiveAimPitch=~
    SkydiveAimYaw=~
    DeployChuteAnimRate=~
    SkydiveDriftAnimRate=~
    SkydiveDriftAnimRateCurrent=~
    SkydiveFidgetAnimRate=~
    SkydiveFidgetAnimRateCurrent=~
    SkydiveAdditiveAlpha=~
    SkydiveDriftDelay=~
   )
   n2 = Sequence Player 'Innovator_Female_Into_FrontEnd_GLIDER'() {
    bLoopAnimation=false
    sequence=&/Innovator_Female_Into_FrontEnd_GLIDER
   }
   n3 = Get OwnerGlider()
   Output Animation Pose(Result=Blend Poses by bool(
     BlendPose_0=@n2
     BlendPose_1=Blendspace Player 'BS_Innovator_Female_Into_GLIDER'(X=@n1.SkydiveAimYaw, Y=@n1.SkydiveAimPitch) {bLoop=false, blendSpace=&/BS_Innovator_Female_Into_GLIDER}
     bActiveValue=Get IsCosmeticPreview(self=@n3)
    ) {blendType=Linear})
  ]
 }
 Closed {
  path=AnimGraph.AnimGraphNode_StateMachine_0.Locomotion.AnimStateNode_2.Closed
  nodes [
   n1 = Get OwnerGlider()
   Output Animation Pose(Result=Blend Poses by bool(
     BlendPose_0=Sequence Player 'Innovator_Female_Closed_FrontEnd_GLIDER'() {sequence=&/Innovator_Female_Closed_FrontEnd_GLIDER}
     BlendPose_1=Sequence Player 'Innovator_Female_Closed_GLIDER'() {sequence=&/Innovator_Female_Closed_GLIDER}
     bActiveValue=Get IsCosmeticPreview(self=@n1)
    ) {blendType=Linear})
  ]
 }
 Transition {
  path=AnimGraph.AnimGraphNode_StateMachine_0.Locomotion.AnimStateTransitionNode_0.Transition
  nodes [
   Result()
   "Equal (Float)"!(A:double="Get Relevant Anim Time Remaining (Open_Into)"!() -> (ReturnValue:float) {is=title K2Node_AnimGetter}, B:double=0) -> (ReturnValue:bool) {is=title K2Node_CallFunction}
  ]
 }
 Transition {
  path=AnimGraph.AnimGraphNode_StateMachine_0.Locomotion.AnimStateTransitionNode_1.Transition
  nodes [
   n1 = Get SkydivingState()
   Result(bCanEnterTransition=OR Boolean(
     A=Break Fort Anim Input Skydiving(FortAnimInput_Skydiving=@n1) -> (
      ParaGilderRootOffset=~
      ParaGliderRootRotationOverride=~
      bUseParaGlideRootModifier=~
      bIsSkydivingFromLaunchPad=~
      bIsSkydivingFromBus=~
      bIsInVortex=~
      bIsDiving=~
      bIsParachuteOpen=~
      bIsSkydiving=~
      bIsParachuteLeaning=~
      bIsSkydiveLeaning=~
      bIsLeaning=~
      bIsSkydiveDiveMode=~
      bPlaySkydiveDrift=~
      bSkydiveDriftDelayActive=~
      bSkydiveDriftAnimAllowed=~
      bIsGliderRight=~
      bIsGliderCenter=~
      bIsGliderLeft=~
      bIsGliderForward=~
      bIsGliderBack=~
      bHasDeployedAndIsStandingUpright=~
      LocalAccelForward=~
      LocalAccelRight=~
      LocalVelocityRight=~
      SkydiveAimPitch=~
      SkydiveAimYaw=~
      DeployChuteAnimRate=~
      SkydiveDriftAnimRate=~
      SkydiveDriftAnimRateCurrent=~
      SkydiveFidgetAnimRate=~
      SkydiveFidgetAnimRateCurrent=~
      SkydiveAdditiveAlpha=~
      SkydiveDriftDelay=~
     ).bIsParachuteOpen
     B=Get bForceOpen() -> (bForceOpen:bool)
    ))
  ]
 }
 Transition {
  path=AnimGraph.AnimGraphNode_StateMachine_0.Locomotion.AnimStateTransitionNode_2.Transition
  nodes [
   n1 = Get bForceOpen()
   n2 = Break Fort Anim Input Skydiving(FortAnimInput_Skydiving=Get SkydivingState()) -> (
    ParaGilderRootOffset=~
    ParaGliderRootRotationOverride=~
    bUseParaGlideRootModifier=~
    bIsSkydivingFromLaunchPad=~
    bIsSkydivingFromBus=~
    bIsInVortex=~
    bIsDiving=~
    bIsParachuteOpen=~
    bIsSkydiving=~
    bIsParachuteLeaning=~
    bIsSkydiveLeaning=~
    bIsLeaning=~
    bIsSkydiveDiveMode=~
    bPlaySkydiveDrift=~
    bSkydiveDriftDelayActive=~
    bSkydiveDriftAnimAllowed=~
    bIsGliderRight=~
    bIsGliderCenter=~
    bIsGliderLeft=~
    bIsGliderForward=~
    bIsGliderBack=~
    bHasDeployedAndIsStandingUpright=~
    LocalAccelForward=~
    LocalAccelRight=~
    LocalVelocityRight=~
    SkydiveAimPitch=~
    SkydiveAimYaw=~
    DeployChuteAnimRate=~
    SkydiveDriftAnimRate=~
    SkydiveDriftAnimRateCurrent=~
    SkydiveFidgetAnimRate=~
    SkydiveFidgetAnimRateCurrent=~
    SkydiveAdditiveAlpha=~
    SkydiveDriftDelay=~
   )
   Result(bCanEnterTransition=AND Boolean(A=NOT Boolean(A=@n2.bIsParachuteOpen), B=NOT Boolean(A=@n1)))
  ]
 }
 Transition {
  path=AnimGraph.AnimGraphNode_StateMachine_0.Locomotion.AnimStateTransitionNode_3.Transition
  nodes [
   n1 = Get bForceOpen()
   n2 = Break Fort Anim Input Skydiving(FortAnimInput_Skydiving=Get SkydivingState()) -> (
    ParaGilderRootOffset=~
    ParaGliderRootRotationOverride=~
    bUseParaGlideRootModifier=~
    bIsSkydivingFromLaunchPad=~
    bIsSkydivingFromBus=~
    bIsInVortex=~
    bIsDiving=~
    bIsParachuteOpen=~
    bIsSkydiving=~
    bIsParachuteLeaning=~
    bIsSkydiveLeaning=~
    bIsLeaning=~
    bIsSkydiveDiveMode=~
    bPlaySkydiveDrift=~
    bSkydiveDriftDelayActive=~
    bSkydiveDriftAnimAllowed=~
    bIsGliderRight=~
    bIsGliderCenter=~
    bIsGliderLeft=~
    bIsGliderForward=~
    bIsGliderBack=~
    bHasDeployedAndIsStandingUpright=~
    LocalAccelForward=~
    LocalAccelRight=~
    LocalVelocityRight=~
    SkydiveAimPitch=~
    SkydiveAimYaw=~
    DeployChuteAnimRate=~
    SkydiveDriftAnimRate=~
    SkydiveDriftAnimRateCurrent=~
    SkydiveFidgetAnimRate=~
    SkydiveFidgetAnimRateCurrent=~
    SkydiveAdditiveAlpha=~
    SkydiveDriftDelay=~
   )
   Result(bCanEnterTransition=AND Boolean(A=NOT Boolean(A=@n2.bIsParachuteOpen), B=NOT Boolean(A=@n1)))
  ]
 }
 Lean {
  entry=Lean_Zero
  path=AnimGraph.AnimGraphNode_StateMachine_1.Lean
  ids=id
  nodes [
   AnimStateNode_0 = Lean_Zero() {is=title AnimStateNode}
   AnimStateNode_1 = Lean_Into() {is=title AnimStateNode}
   AnimStateNode_2 = Lean_Out() {is=title AnimStateNode}
   AnimStateTransitionNode_0 = AnimStateTransitionNode() {blendMode=Linear, crossfadeDuration=0.05}
   AnimStateTransitionNode_1 = AnimStateTransitionNode() {blendMode=Linear, crossfadeDuration=0.1}
   AnimStateTransitionNode_2 = AnimStateTransitionNode() {bAutomaticRuleBasedOnSequencePlayerInState=true, blendMode=Linear, priorityOrder=2}
   AnimStateTransitionNode_3 = AnimStateTransitionNode() {blendMode=Linear}
  ]
  links [
   {
    dst=Lean_Out
    rule=AnimGraph.AnimGraphNode_StateMachine_1.Lean.AnimStateTransitionNode_1.Transition
    src=Lean_Into
    via=AnimStateTransitionNode_1
   }
   {
    dst=Lean_Zero
    rule=AnimGraph.AnimGraphNode_StateMachine_1.Lean.AnimStateTransitionNode_2.Transition
    src=Lean_Out
    via=AnimStateTransitionNode_2
   }
   {
    dst=Lean_Zero
    rule=AnimGraph.AnimGraphNode_StateMachine_1.Lean.AnimStateTransitionNode_3.Transition
    src=Lean_Out
    via=AnimStateTransitionNode_3
   }
   {
    dst=Lean_Into
    rule=AnimGraph.AnimGraphNode_StateMachine_1.Lean.AnimStateTransitionNode_0.Transition
    src=Lean_Zero
    via=AnimStateTransitionNode_0
   }
  ]
 }
 Lean_Zero {
  path=AnimGraph.AnimGraphNode_StateMachine_1.Lean.AnimStateNode_0.Lean_Zero
  nodes [Output Animation Pose(Result="Sequence Player (additive) 'Innovator_Female_Lean_Zero_IO_Anim_GLIDER'"() {sequence=&/Innovator_Female_Lean_Zero_IO_Anim_GLIDER})]
 }
 Lean_Into {
  path=AnimGraph.AnimGraphNode_StateMachine_1.Lean.AnimStateNode_1.Lean_Into
  nodes [
   n1 = Break Fort Anim Input Skydiving(FortAnimInput_Skydiving=Get SkydivingState()) -> (
    ParaGilderRootOffset=~
    ParaGliderRootRotationOverride=~
    bUseParaGlideRootModifier=~
    bIsSkydivingFromLaunchPad=~
    bIsSkydivingFromBus=~
    bIsInVortex=~
    bIsDiving=~
    bIsParachuteOpen=~
    bIsSkydiving=~
    bIsParachuteLeaning=~
    bIsSkydiveLeaning=~
    bIsLeaning=~
    bIsSkydiveDiveMode=~
    bPlaySkydiveDrift=~
    bSkydiveDriftDelayActive=~
    bSkydiveDriftAnimAllowed=~
    bIsGliderRight=~
    bIsGliderCenter=~
    bIsGliderLeft=~
    bIsGliderForward=~
    bIsGliderBack=~
    bHasDeployedAndIsStandingUpright=~
    LocalAccelForward=~
    LocalAccelRight=~
    LocalVelocityRight=~
    SkydiveAimPitch=~
    SkydiveAimYaw=~
    DeployChuteAnimRate=~
    SkydiveDriftAnimRate=~
    SkydiveDriftAnimRateCurrent=~
    SkydiveFidgetAnimRate=~
    SkydiveFidgetAnimRateCurrent=~
    SkydiveAdditiveAlpha=~
    SkydiveDriftDelay=~
   )
   "Blend Poses (ESkydivingDirection)"!(
    BlendPose_0:Pose Link Structure="Sequence Player (additive) 'Grim_Male_Lean_Zero_IO_GLIDER'"() {bLoopAnimation=false, sequence=/BRCosmetics/Animation/Game/MainPlayer/Skydive/ParaGlide/Grim_Male/Grim_Male_Lean_Zero_IO_GLIDER}
    BlendPose_1:Pose Link Structure="Sequence Player (additive) 'Innovator_Female_Lean_Forward_Into_GLIDER'"() {sequence=&/Innovator_Female_Lean_Forward_Into_GLIDER}
    BlendPose_2:Pose Link Structure="Sequence Player (additive) 'Innovator_Female_Lean_Back_Into_GLIDER'"() {sequence=&/Innovator_Female_Lean_Back_Into_GLIDER}
    BlendPose_3:Pose Link Structure="Sequence Player (additive) 'Innovator_Female_Lean_Right_Into_GLIDER'"() {sequence=&/Innovator_Female_Lean_Right_Into_GLIDER}
    BlendPose_4:Pose Link Structure="Sequence Player (additive) 'Innovator_Female_Lean_Left_Into_GLIDER'"() {sequence=&/Innovator_Female_Lean_Left_Into_GLIDER}
    BlendTime_0:float=0.5
    BlendTime_1:float=0.5
    BlendTime_2:float=0.5
    BlendTime_3:float=0.5
    BlendTime_4:float=0.5
    ActiveEnumValue:Byte=@n1.LocalAccelDir
   ) -> (Pose:Pose Link Structure) {
    is=title AnimGraphNode_BlendListByEnum
    blendType=Linear
    blendPose [
     {linkId=-1, sourceLinkId=-1}
     {linkId=-1, sourceLinkId=-1}
     {linkId=-1, sourceLinkId=-1}
     {linkId=-1, sourceLinkId=-1}
     {linkId=-1, sourceLinkId=-1}
    ]
    blendTime [0.1,0.1,0.1,0.1,0.1]
   }
   Sequence Player!(Sequence=Get LeanIntoTransitionAnim() -> (LeanIntoTransitionAnim:Anim Sequence Base Object Reference)) -> (Pose) {bLoopAnimation=false}
   n2 = Break Fort Anim Input Skydiving(FortAnimInput_Skydiving=Get SkydivingState()) -> (
    ParaGilderRootOffset=~
    ParaGliderRootRotationOverride=~
    bUseParaGlideRootModifier=~
    bIsSkydivingFromLaunchPad=~
    bIsSkydivingFromBus=~
    bIsInVortex=~
    bIsDiving=~
    bIsParachuteOpen=~
    bIsSkydiving=~
    bIsParachuteLeaning=~
    bIsSkydiveLeaning=~
    bIsLeaning=~
    bIsSkydiveDiveMode=~
    bPlaySkydiveDrift=~
    bSkydiveDriftDelayActive=~
    bSkydiveDriftAnimAllowed=~
    bIsGliderRight=~
    bIsGliderCenter=~
    bIsGliderLeft=~
    bIsGliderForward=~
    bIsGliderBack=~
    bHasDeployedAndIsStandingUpright=~
    LocalAccelForward=~
    LocalAccelRight=~
    LocalVelocityRight=~
    SkydiveAimPitch=~
    SkydiveAimYaw=~
    DeployChuteAnimRate=~
    SkydiveDriftAnimRate=~
    SkydiveDriftAnimRateCurrent=~
    SkydiveFidgetAnimRate=~
    SkydiveFidgetAnimRateCurrent=~
    SkydiveAdditiveAlpha=~
    SkydiveDriftDelay=~
   )
   n3 = Blendspace Player 'BS_Innovator_Female_Lean_Transition_GLIDER'(X=@n1.LocalAccelRight, Y=@n1.LocalAccelForward) {bLoop=false, blendSpace=&/BS_Innovator_Female_Lean_Transition_GLIDER}
   Output Animation Pose(Result=Two Way Blend!(
     A=Blendspace Player 'BS_Innovator_Female_Lean_GLIDER'(X=@n2.LocalAccelRight, Y=@n2.LocalAccelForward) {blendSpace=&BS_Innovator_Female_Lean_GLIDER}
     B=@n3
     AlphaCurveName:Name=GliderLeanTransition
    ) -> (Pose) {blendNode {alphaInputType=Curve, alphaBoolBlend {blendOption=Linear}, alphaScaleBiasClamp {bias=1, scale=-1}}})
   Get ParaGlideLeanTransition() -> (ParaGlideLeanTransition:double)
  ]
 }
 Lean_Out {
  path=AnimGraph.AnimGraphNode_StateMachine_1.Lean.AnimStateNode_2.Lean_Out
  nodes [
   n1 = "Blend Poses (ESkydivingDirection)"(
    BlendPose_0="Sequence Player (additive) 'Innovator_Female_Lean_Center_Pose_GLIDER'"() {sequence=&/Innovator_Female_Lean_Center_Pose_GLIDER}
    BlendPose_1="Sequence Player (additive) 'Innovator_Female_Lean_Forward_Out_GLIDER'"() {sequence=&/Innovator_Female_Lean_Forward_Out_GLIDER}
    BlendPose_2="Sequence Player (additive) 'Innovator_Female_Lean_Back_Out_GLIDER'"() {sequence=&/Innovator_Female_Lean_Back_Out_GLIDER}
    BlendPose_3="Sequence Player (additive) 'Innovator_Female_Lean_Left_Out_GLIDER'"() {sequence=&/Innovator_Female_Lean_Left_Out_GLIDER}
    BlendPose_4="Sequence Player (additive) 'Innovator_Female_Lean_Right_Out_GLIDER'"() {sequence=&/Innovator_Female_Lean_Right_Out_GLIDER}
    BlendTime_0=0.5
    BlendTime_1=0.5
    BlendTime_2=0.5
    BlendTime_3=0.5
    BlendTime_4=0.5
    ActiveEnumValue=Break Fort Anim Input Skydiving(FortAnimInput_Skydiving=Get SkydivingState()) -> (
     ParaGilderRootOffset=~
     ParaGliderRootRotationOverride=~
     bUseParaGlideRootModifier=~
     bIsSkydivingFromLaunchPad=~
     bIsSkydivingFromBus=~
     bIsInVortex=~
     bIsDiving=~
     bIsParachuteOpen=~
     bIsSkydiving=~
     bIsParachuteLeaning=~
     bIsSkydiveLeaning=~
     bIsLeaning=~
     bIsSkydiveDiveMode=~
     bPlaySkydiveDrift=~
     bSkydiveDriftDelayActive=~
     bSkydiveDriftAnimAllowed=~
     bIsGliderRight=~
     bIsGliderCenter=~
     bIsGliderLeft=~
     bIsGliderForward=~
     bIsGliderBack=~
     bHasDeployedAndIsStandingUpright=~
     LocalAccelForward=~
     LocalAccelRight=~
     LocalVelocityRight=~
     SkydiveAimPitch=~
     SkydiveAimYaw=~
     DeployChuteAnimRate=~
     SkydiveDriftAnimRate=~
     SkydiveDriftAnimRateCurrent=~
     SkydiveFidgetAnimRate=~
     SkydiveFidgetAnimRateCurrent=~
     SkydiveAdditiveAlpha=~
     SkydiveDriftDelay=~
    ).DirectionLast
   ) {
    blendType=Linear
    blendPose [
     {linkId=-1, sourceLinkId=-1}
     {linkId=-1, sourceLinkId=-1}
     {linkId=-1, sourceLinkId=-1}
     {linkId=-1, sourceLinkId=-1}
     {linkId=-1, sourceLinkId=-1}
    ]
    blendTime [0.1,0.1,0.1,0.1,0.1]
   }
   Output Animation Pose(Result=@n1)
  ]
 }
 Transition {
  path=AnimGraph.AnimGraphNode_StateMachine_1.Lean.AnimStateTransitionNode_0.Transition
  nodes [Result(bCanEnterTransition=Get bIsAccelerating2D())]
 }
 Transition {
  path=AnimGraph.AnimGraphNode_StateMachine_1.Lean.AnimStateTransitionNode_1.Transition
  nodes [Result(bCanEnterTransition=NOT Boolean(A=Get bIsAccelerating2D()))]
 }
 Transition {path=AnimGraph.AnimGraphNode_StateMachine_1.Lean.AnimStateTransitionNode_2.Transition, nodes [Result()]}
 Transition {
  path=AnimGraph.AnimGraphNode_StateMachine_1.Lean.AnimStateTransitionNode_3.Transition
  nodes [Result(bCanEnterTransition=Get bIsAccelerating2D())]
 }
 DirectionBIllow {nodes [
   Inputs()
   > n1 = Sequence()
   > Branch(Condition=Get bIsAccelerating2D())
   > Do Once(Reset=@n3.then)
   > Set Billow_Anim(Billow_Anim:Anim Sequence Base Object Reference) -> (Output_Get:Anim Sequence Base Object Reference)
   > Set Billow_Alpha_Target(Billow_Alpha_Target:double=0.5) -> (Output_Get:double)
   > Delay(Duration=0.933)
   > Set Billow_Alpha_Target(Billow_Alpha_Target=0)
   > Outputs!(then:Exec)
   n2 = Get SkydivingState()
   n3 = Branch(
    execute=@n1.then_1
    Condition="Equal (Enum)"(
     A:ESkydivingDirection Enum=Break Fort Anim Input Skydiving(FortAnimInput_Skydiving=@n2) -> (
      ParaGilderRootOffset=~
      ParaGliderRootRotationOverride=~
      bUseParaGlideRootModifier=~
      bIsSkydivingFromLaunchPad=~
      bIsSkydivingFromBus=~
      bIsInVortex=~
      bIsDiving=~
      bIsParachuteOpen=~
      bIsSkydiving=~
      bIsParachuteLeaning=~
      bIsSkydiveLeaning=~
      bIsLeaning=~
      bIsSkydiveDiveMode=~
      bPlaySkydiveDrift=~
      bSkydiveDriftDelayActive=~
      bSkydiveDriftAnimAllowed=~
      bIsGliderRight=~
      bIsGliderCenter=~
      bIsGliderLeft=~
      bIsGliderForward=~
      bIsGliderBack=~
      bHasDeployedAndIsStandingUpright=~
      LocalAccelForward=~
      LocalAccelRight=~
      LocalVelocityRight=~
      SkydiveAimPitch=~
      SkydiveAimYaw=~
      DeployChuteAnimRate=~
      SkydiveDriftAnimRate=~
      SkydiveDriftAnimRateCurrent=~
      SkydiveFidgetAnimRate=~
      SkydiveFidgetAnimRateCurrent=~
      SkydiveAdditiveAlpha=~
      SkydiveDriftDelay=~
     ).LocalAccelDir
     B:ESkydivingDirection Enum=Center
    )
   )
  ]}
 EventGraph {nodes [
   n1 = Event BlueprintUpdateAnimation()
   > Set Time(Time:double=@n1.DeltaTimeX) -> (Output_Get:double)
   > Set Lean Into Anim!(execute:Exec) -> (then:Exec) {is=title K2Node_MacroInstance}
   n2 = Break Fort Anim Input Skydiving(FortAnimInput_Skydiving=Get SkydivingState()) -> (
    ParaGilderRootOffset=~
    ParaGliderRootRotationOverride=~
    bUseParaGlideRootModifier=~
    bIsSkydivingFromLaunchPad=~
    bIsSkydivingFromBus=~
    bIsInVortex=~
    bIsDiving=~
    bIsParachuteOpen=~
    bIsSkydiving=~
    bIsParachuteLeaning=~
    bIsSkydiveLeaning=~
    bIsLeaning=~
    bIsSkydiveDiveMode=~
    bPlaySkydiveDrift=~
    bSkydiveDriftDelayActive=~
    bSkydiveDriftAnimAllowed=~
    bIsGliderRight=~
    bIsGliderCenter=~
    bIsGliderLeft=~
    bIsGliderForward=~
    bIsGliderBack=~
    bHasDeployedAndIsStandingUpright=~
    LocalAccelForward=~
    LocalAccelRight=~
    LocalVelocityRight=~
    SkydiveAimPitch=~
    SkydiveAimYaw=~
    DeployChuteAnimRate=~
    SkydiveDriftAnimRate=~
    SkydiveDriftAnimRateCurrent=~
    SkydiveFidgetAnimRate=~
    SkydiveFidgetAnimRateCurrent=~
    SkydiveAdditiveAlpha=~
    SkydiveDriftDelay=~
   )
   > n3 = ForwardBack_Value_Interp!(
    execute
    Time:double=Get Time() -> (Time:double)
    Currnet_Value:double=Get Jitter_Speed()
    Idle_Value:double=1
    Idle_InterpSpeed:double=1
    Forward_Value:double=1.5
    Forward_InterpSpeed:double=2
    Back_Value:double=1
    Back_InterpSpeed:double=2
    LocalAccelForward:double=@n2.LocalAccelForward
   ) -> (then, New_Value:double)
   > Set Jitter_Speed(Jitter_Speed:double=@n3) -> (Output_Get:double)
   > n4 = ForwardBack_Value_Interp(
    Time=Get Time()
    Currnet_Value=Get Billow_Speed()
    Idle_Value=1
    Idle_InterpSpeed=1
    Forward_Value=2
    Forward_InterpSpeed=2
    Back_Value=1
    Back_InterpSpeed=2
    LocalAccelForward=@n2.LocalAccelForward
   )
   > Set Billow_Speed(Billow_Speed:double=@n4) -> (Output_Get:double)
   > n5 = Sequence!(execute) -> (then_0, then_1, then_2:Exec)
   > Random Billow!(execute:Exec) -> (then:Exec) {is=title K2Node_MacroInstance}
   Direction BIllow!(execute:Exec=@n5.then_1) -> (then:Exec) {is=title K2Node_MacroInstance}
   Return to Idle BIllow!(execute:Exec=@n5.then_2) -> (then:Exec) {is=title K2Node_MacroInstance}
   AnimNotify_TurnONEngines() {is=title K2Node_Event}
   > SetGliderFullyDeployed!(execute, self:Fort Glider Anim Instance Object Reference, bIsFullyDeployed:bool=true) -> (then)
   AnimNotify_DeployEntered() {is=title K2Node_Event}
   > Delay(Duration=0.15)
   n6 = Get OwnerGlider()
   > Set Actor Hidden In Game!(execute, self:Actor Object Reference=@n6, bNewHidden:bool=false) -> (then) {is=title K2Node_CallFunction}
   AnimNotify_ClosedEntered() {is=title K2Node_Event}
   > Set Actor Hidden In Game(self=@n6, bNewHidden=true)
   Break Fort Anim Input Skydiving(FortAnimInput_Skydiving=Get SkydivingState()) -> (
    ParaGilderRootOffset=~
    ParaGliderRootRotationOverride=~
    bUseParaGlideRootModifier=~
    bIsSkydivingFromLaunchPad=~
    bIsSkydivingFromBus=~
    bIsInVortex=~
    bIsDiving=~
    bIsParachuteOpen=~
    bIsSkydiving=~
    bIsParachuteLeaning=~
    bIsSkydiveLeaning=~
    bIsLeaning=~
    bIsSkydiveDiveMode=~
    bPlaySkydiveDrift=~
    bSkydiveDriftDelayActive=~
    bSkydiveDriftAnimAllowed=~
    bIsGliderRight=~
    bIsGliderCenter=~
    bIsGliderLeft=~
    bIsGliderForward=~
    bIsGliderBack=~
    bHasDeployedAndIsStandingUpright=~
    LocalAccelForward=~
    LocalAccelRight=~
    LocalVelocityRight=~
    SkydiveAimPitch=~
    SkydiveAimYaw=~
    DeployChuteAnimRate=~
    SkydiveDriftAnimRate=~
    SkydiveDriftAnimRateCurrent=~
    SkydiveFidgetAnimRate=~
    SkydiveFidgetAnimRateCurrent=~
    SkydiveAdditiveAlpha=~
    SkydiveDriftDelay=~
   )
  ]}
 HideNShow {nodes [
   Inputs()
   Outputs()
  ]}
 RandomBillow {nodes [
   Inputs()
   > n1 = Sequence()
   > Do Once(Reset=@n2.then)
   > Set Billow_Anim()
   > Set Billow_Alpha_Target(Billow_Alpha_Target=RandomFloatInRange(Min=0.8, Max=1))
   > Delay(Duration=float / float(A=2, B=Get Billow_Speed()) {is=title K2Node_CallFunction})
   > Set Billow_Alpha_Target(Billow_Alpha_Target=0)
   > Outputs()
   n2 = Delay(execute=@n1.then_1, Duration=RandomFloatInRange(Min=4, Max=7))
   Set Billow_Alpha(
    execute=@n1.then_2
    Billow_Alpha:double=FInterpTo(Current=Get Billow_Alpha(), Target=Get Billow_Alpha_Target() -> (Billow_Alpha_Target:double), DeltaTime=Get Time(), InterpSpeed=2)
   ) -> (Output_Get:double)
   Select(Option 0:Wildcard, Option 1:Wildcard, Index=RandomBool!() -> (ReturnValue:bool)) -> (ReturnValue:Wildcard)
  ]}
 ReturnToIdleBIllow {nodes [
   Inputs()
   > n1 = Sequence!(execute) -> (then_0, then_1)
   n2 = Get bIsAccelerating2D()
   n3 = Break Fort Anim Input Skydiving(FortAnimInput_Skydiving=Get SkydivingState()) -> (
    ParaGilderRootOffset=~
    ParaGliderRootRotationOverride=~
    bUseParaGlideRootModifier=~
    bIsSkydivingFromLaunchPad=~
    bIsSkydivingFromBus=~
    bIsInVortex=~
    bIsDiving=~
    bIsParachuteOpen=~
    bIsSkydiving=~
    bIsParachuteLeaning=~
    bIsSkydiveLeaning=~
    bIsLeaning=~
    bIsSkydiveDiveMode=~
    bPlaySkydiveDrift=~
    bSkydiveDriftDelayActive=~
    bSkydiveDriftAnimAllowed=~
    bIsGliderRight=~
    bIsGliderCenter=~
    bIsGliderLeft=~
    bIsGliderForward=~
    bIsGliderBack=~
    bHasDeployedAndIsStandingUpright=~
    LocalAccelForward=~
    LocalAccelRight=~
    LocalVelocityRight=~
    SkydiveAimPitch=~
    SkydiveAimYaw=~
    DeployChuteAnimRate=~
    SkydiveDriftAnimRate=~
    SkydiveDriftAnimRateCurrent=~
    SkydiveFidgetAnimRate=~
    SkydiveFidgetAnimRateCurrent=~
    SkydiveAdditiveAlpha=~
    SkydiveDriftDelay=~
   )
   > Branch(Condition=AND Boolean(
     A="Not Equal (Enum)"!(A:ESkydivingDirection Enum=@n3.DirectionLast, B:ESkydivingDirection Enum=Center) -> (ReturnValue:bool)
     B=NOT Boolean(A=@n2)
    ))
   > Do Once(Reset=@n4.then)
   > Set Billow_Anim()
   > Set Billow_Alpha_Target(Billow_Alpha_Target=1)
   > Delay(Duration=0.933)
   > Set Billow_Alpha_Target(Billow_Alpha_Target=0)
   > Outputs()
   n4 = Branch(execute=@n1.then_1, Condition=Get bIsAccelerating2D())
   "Equal (Enum)"(
    A=Break Fort Anim Input Skydiving(FortAnimInput_Skydiving=Get SkydivingState()) -> (
     ParaGilderRootOffset=~
     ParaGliderRootRotationOverride=~
     bUseParaGlideRootModifier=~
     bIsSkydivingFromLaunchPad=~
     bIsSkydivingFromBus=~
     bIsInVortex=~
     bIsDiving=~
     bIsParachuteOpen=~
     bIsSkydiving=~
     bIsParachuteLeaning=~
     bIsSkydiveLeaning=~
     bIsLeaning=~
     bIsSkydiveDiveMode=~
     bPlaySkydiveDrift=~
     bSkydiveDriftDelayActive=~
     bSkydiveDriftAnimAllowed=~
     bIsGliderRight=~
     bIsGliderCenter=~
     bIsGliderLeft=~
     bIsGliderForward=~
     bIsGliderBack=~
     bHasDeployedAndIsStandingUpright=~
     LocalAccelForward=~
     LocalAccelRight=~
     LocalVelocityRight=~
     SkydiveAimPitch=~
     SkydiveAimYaw=~
     DeployChuteAnimRate=~
     SkydiveDriftAnimRate=~
     SkydiveDriftAnimRateCurrent=~
     SkydiveFidgetAnimRate=~
     SkydiveFidgetAnimRateCurrent=~
     SkydiveAdditiveAlpha=~
     SkydiveDriftDelay=~
    ).LocalAccelDir
    B=Center
   )
  ]}
 SetLeanIntoAnim {nodes [
   n1 = Break Fort Anim Input Skydiving(FortAnimInput_Skydiving=Get SkydivingState()) -> (
    ParaGilderRootOffset=~
    ParaGliderRootRotationOverride=~
    bUseParaGlideRootModifier=~
    bIsSkydivingFromLaunchPad=~
    bIsSkydivingFromBus=~
    bIsInVortex=~
    bIsDiving=~
    bIsParachuteOpen=~
    bIsSkydiving=~
    bIsParachuteLeaning=~
    bIsSkydiveLeaning=~
    bIsLeaning=~
    bIsSkydiveDiveMode=~
    bPlaySkydiveDrift=~
    bSkydiveDriftDelayActive=~
    bSkydiveDriftAnimAllowed=~
    bIsGliderRight=~
    bIsGliderCenter=~
    bIsGliderLeft=~
    bIsGliderForward=~
    bIsGliderBack=~
    bHasDeployedAndIsStandingUpright=~
    LocalAccelForward=~
    LocalAccelRight=~
    LocalVelocityRight=~
    SkydiveAimPitch=~
    SkydiveAimYaw=~
    DeployChuteAnimRate=~
    SkydiveDriftAnimRate=~
    SkydiveDriftAnimRateCurrent=~
    SkydiveFidgetAnimRate=~
    SkydiveFidgetAnimRateCurrent=~
    SkydiveAdditiveAlpha=~
    SkydiveDriftDelay=~
   )
   n2 = Branch(Condition="float >= float"(A=@n1.LocalAccelForward, B=250) {is=title K2Node_CallFunction})
   > n3 = Set LeanIntoTransitionAnim(LeanIntoTransitionAnim:Anim Sequence Base Object Reference) -> (Output_Get:Anim Sequence Base Object Reference)
   Outputs(then=[@n3.then, @n6.then, @n5.then, @n7.then])
   n4 = Branch(
    execute=@n2.else
    Condition="float <= float"!(A:double=@n1.LocalAccelForward, B:double=250) -> (ReturnValue:bool) {is=title K2Node_CallFunction}
   )
   > n5 = Set LeanIntoTransitionAnim()
   n6 = Set LeanIntoTransitionAnim(execute=@n4.else, LeanIntoTransitionAnim=&Grim_Male_Lean_Zero_IO_GLIDER)
   Inputs()
   > n7 = Set LeanIntoTransitionAnim(LeanIntoTransitionAnim=Select!(
     Center:Anim Sequence Object Reference=&Grim_Male_Lean_Zero_IO_GLIDER
     Right:Anim Sequence Object Reference
     Left:Anim Sequence Object Reference
     Forward:Anim Sequence Object Reference
     Back:Anim Sequence Object Reference
     Index:ESkydivingDirection Enum=@n1.DirectionLast
    ) -> (ReturnValue:Anim Sequence Object Reference))
  ]}
]
variables [
 Billow_Alpha {type=number, default=0.1313}
 Billow_Alpha_Target {type=number}
 Billow_Anim {type=AnimSequenceBase, default=null}
 Billow_Speed {type=number, default=1}
 Jitter_Speed {type=number, default=1}
 LeanIntoTransitionAnim {type=AnimSequenceBase, default=&Grim_Male_Lean_Zero_IO_GLIDER}
 ParaGlideLeanTransition {type=number}
 Time {type=number, default=0.0167}
]
